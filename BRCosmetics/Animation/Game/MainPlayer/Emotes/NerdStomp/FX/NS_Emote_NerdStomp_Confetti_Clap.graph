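# Niagara effect for the NerdStomp emote, named for a clap, that fires a multicolor confetti pop with flash and spikes.
# Burst layers flashes, sparks and colored bits in one emitter, using spawn-group masks and visibility tags across four sprite renderers.
# Confetti fires quick looped bursts of palette-colored, colliding sprites with drag, gravity and curl noise, plus a front-end color offset.
# SpikeBurst adds short-lived sprite and mesh streaks with panning material and distance culling.
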NiagaraSystem /BRCosmetics/Animation/Game/MainPlayer/Emotes/NerdStomp/FX/NS_Emote_NerdStomp_Confetti_Clap
dependencies {
 usedDataInterfaces [
  &NiagaraDataInterfaceArrayColor
  &NiagaraDataInterfaceCurve
  &NiagaraDataInterfaceArrayFloat
  &NiagaraDataInterfaceVectorCurve
 ]
 usedDynamicInputs []
 usedModules [&SystemState]
}
emitters {
 Burst {
  emitterUpdateScript [
   &EmitterState()
   &SpawnBurst_Instantaneous(Spawn Count=&RandomRangeInt(Minimum=25, Maximum=50), Spawn Group=3, Use Spawn Probability=false) {module=SpawnBurst_Instantaneous003}
   &SpawnBurst_Instantaneous(Spawn Count=1, Use Spawn Probability=false)
   &SpawnBurst_Instantaneous(Spawn Count=1, Spawn Group=1, Use Spawn Probability=false) {module=SpawnBurst_Instantaneous001}
   &SpawnBurst_Instantaneous(
    Spawn Count=&RandomRangeInt(Minimum=24, Maximum=34)
    Spawn Group=2
    Use Spawn Probability=false
    UseLoopCountLimit=false
   ) {module=SpawnBurst_Instantaneous002}
  ]
  particleSpawnScript [
   /Niagara/Modules/Spawn/Initialization/InitializeParticle(
    Color=<rgba 0.9 0.6298 0.1365 1>
    Lifetime=0.2
    Mesh Scale=<1.5 1.5 1.5>
    Module.Write Color=true
    Module.Write Lifetime=true
    Module.Write Mass=true
    Module.Write Material Random=true
    Module.Write Position=false
    Module.Write Scale=false
    Module.Write SpriteRotation=true
    Module.Write SpriteSize=true
    Position=&SimulationPosition()
    Sprite Rotation=&UniformRangedFloat(Minimum=0, Maximum=1)
    Sprite Size=<30 30>
   ) {enabled=false}
   /SetVariables(Particles.VisibilityTag=&SpawnGroup()) {isSetParameters=true}
   /Niagara/Modules/Spawn/Initialization/V2/InitializeParticle(
    Color=&MaskLinearColorBySpawnGroup(
     Masked Linear Color=<rgba 1 1 1 1>
     Passthrough Linear Color=&MaskLinearColorBySpawnGroup(
      Masked Linear Color=<rgba 2 1.7057 0.9688 1>
      Passthrough Linear Color=&MaskLinearColorBySpawnGroup(
       Masked Linear Color=&SelectLinearColorFromArray(
        Array Sampling Mode=::NewEnumerator0
        Direct Array Index=?
        Color Selection Array=DataInterface() {propertyValues {ColorData [<rgba 1 0.1614 0.9113 1>, <rgba 0.148 0 1 1>, <rgba 0.0772 0.9637 1 1>, <rgba 1 0.9088 0.0365 1>, <rgba 0.9385 0 0 1>]}}
        Random Seed=0
       )
       Passthrough Linear Color=&SelectLinearColorFromArray(
        Array Sampling Mode=::NewEnumerator0
        Direct Array Index=?
        Color Selection Array=DataInterface() {propertyValues {ColorData [<rgba 1 0.1614 0.9113 1>, <rgba 0.148 0 1 1>, <rgba 0.0772 0.9637 1 1>, <rgba 1 0.9088 0.0365 1>, <rgba 0.9385 0 0 1>]}}
        Random Seed=0
       )
       Spawn Group Mask=2
      )
      Spawn Group Mask=1
     )
    )
    Lifetime=&MaskFloatBySpawnGroup(
     Masked Float=0.3
     Passthrough Float=&MaskFloatBySpawnGroup(
      Masked Float=0.25
      Passthrough Float=&MaskFloatBySpawnGroup(
       Masked Float=&Lerp_Float(
        A=0.25
        B=1
        Alpha=&FloatFromCurve(
         FloatCurve=DataInterface() {propertyValues {Curve {keys [
             {interpMode=RCIM_Cubic, tangentMode=RCTM_User, time=0, value=0}
             {arriveTangent=2.9148, interpMode=RCIM_Cubic, leaveTangent=2.9148, tangentMode=RCTM_User, time=1, value=1}
            ]}}}
         CurveIndex=&RandomRangeFloat(Minimum=0, Maximum=1)
        )
       )
       Passthrough Float=&RandomRangeFloat(Minimum=0.125, Maximum=0.35)
       Spawn Group Mask=2
      )
      Spawn Group Mask=1
     )
    )
    Sprite Rotation Mode=::NewEnumerator2
    Sprite Size=&MaskVector2dBySpawnGroup(
     Masked Vector2D=<65 65>
     Passthrough Vector2D=&MaskVector2dBySpawnGroup(
      Masked Vector2D=<150 150>
      Passthrough Vector2D=&MaskVector2dBySpawnGroup(
       Masked Vector2D=&Multiply_Vector2D_ByFloat(
        Float=1
        Vector2D=&MakeVector2D(
         X=&RandomRangeFloat(Minimum=8, Maximum=15)
         Y=&Lerp_Float(
          A=12
          B=50
          Alpha=&FloatFromCurve(
           FloatCurve=DataInterface() {propertyValues {Curve {keys [
               {interpMode=RCIM_Cubic, tangentMode=RCTM_User, time=0, value=0}
               {arriveTangent=1, interpMode=RCIM_Cubic, leaveTangent=1, time=0.8, value=0.3}
               {arriveTangent=6.5099, arriveTangentWeight=0.7128, interpMode=RCIM_Cubic, leaveTangent=6.5099, tangentMode=RCTM_User, time=1, value=1}
              ]}}}
           CurveIndex=&RandomRangeFloat(Minimum=0, Maximum=1)
          )
         )
        )
       )
       Passthrough Vector2D=&RandomRangeVector2D(Minimum=<1 2>, Maximum=<2 2>)
       Spawn Group Mask=2
      )
      Spawn Group Mask=1
     )
     Spawn Group Mask=0
    )
    Sprite Size Mode=::NewEnumerator4
    UsePositionOffset=false
    Write Material Random=true
   ) {module=InitializeParticle001}
   &ShapeLocation(
    MaskSpawnGroup=false
    Override Seed=false
    Sphere Radius=&MaskFloatBySpawnGroup(
     Masked Float=0
     Passthrough Float=&MaskFloatBySpawnGroup(
      Masked Float=0
      Passthrough Float=&MaskFloatBySpawnGroup(Masked Float=13, Passthrough Float=5, Spawn Group Mask=2)
      Spawn Group Mask=1
     )
    )
    Sphere Surface Distribution=1
    UseOwnerScale=true
   )
   &SpriteFacingAndAlignment(Sprite Facing=$Particles.ShapeLocation.ShapeNormal, WriteAlignment=false, WriteFacing=true)
   &AddVelocity(
    Module.Use Falloff Distance=false
    Override Seed=false
    Velocity Mode=::NewEnumerator1
    Velocity Speed=&MaskFloatBySpawnGroup(
     Masked Float=0
     Passthrough Float=&MaskFloatBySpawnGroup(
      Masked Float=0
      Passthrough Float=&MaskFloatBySpawnGroup(
       Masked Float=&Lerp_Float(
        A=50
        B=600
        Alpha=&FloatFromCurve(
         FloatCurve=DataInterface() {propertyValues {Curve {keys [
             {interpMode=RCIM_Cubic, tangentMode=RCTM_User, time=0, value=0}
             {arriveTangent=2.9148, interpMode=RCIM_Cubic, leaveTangent=2.9148, tangentMode=RCTM_User, time=1, value=1}
            ]}}}
         CurveIndex=&RandomRangeFloat(Minimum=0, Maximum=1)
        )
       )
       Passthrough Float=&RandomRangeFloat(Minimum=50, Maximum=300)
       Spawn Group Mask=2
      )
      Spawn Group Mask=1
     )
    )
   )
  ]
  particleUpdateScript [
   &ParticleState()
   &SpriteRotationRate(Rotation Rate=&MaskFloatBySpawnGroup(
     Masked Float=0
     Passthrough Float=&MaskFloatBySpawnGroup(
      Masked Float=0
      Passthrough Float=&MaskFloatBySpawnGroup(
       Masked Float=&Multiply_Float(
        A=&Lerp_Float(
         A=25
         B=1000
         Alpha=&FloatFromCurve(FloatCurve=DataInterface() {propertyValues {Curve {keys [
              {arriveTangent=-3.2581, interpMode=RCIM_Cubic, leaveTangent=-3.2581, tangentMode=RCTM_User, time=0.1, value=1}
              {interpMode=RCIM_Cubic, tangentMode=RCTM_User, time=1, value=0}
             ]}}})
        )
        B=&FloatFromCurve(
         FloatCurve=DataInterface() {propertyValues {Curve {keys [<at 0 -1 RCIM_Constant>,<at 0.5 1 RCIM_Constant>]}}}
         CurveIndex=&RandomRangeFloat(Minimum=-1, Maximum=1)
        )
       )
       Passthrough Float=0
       Spawn Group Mask=2
      )
      Spawn Group Mask=1
     )
    ))
   &Drag(
    Drag=&MaskFloatBySpawnGroup(
     Masked Float=3
     Passthrough Float=&MaskFloatBySpawnGroup(
      Masked Float=3
      Passthrough Float=&MaskFloatBySpawnGroup(Masked Float=8, Passthrough Float=0, Spawn Group Mask=2)
      Spawn Group Mask=1
     )
    )
    Use Linear Drag=true
    Use Rotational Drag=true
   )
   &AccelerationForce(Acceleration=&MaskVectorBySpawnGroup(
     Masked Vector=<0 0 0>
     Passthrough Vector=&MaskVectorBySpawnGroup(
      Masked Vector=<0 0 0>
      Passthrough Vector=&MaskVectorBySpawnGroup(
       Masked Vector=&RandomRangeVector(Minimum=<0 0 30>, Maximum=<-25 -25 -30>)
       Passthrough Vector=<0 0 -100>
       Spawn Group Mask=2
      )
      Spawn Group Mask=1
     )
    ))
   &CurlNoiseForce(
    Noise Strength=&MaskFloatBySpawnGroup(
     Masked Float=0
     Passthrough Float=&MaskFloatBySpawnGroup(
      Masked Float=0
      Passthrough Float=&MaskFloatBySpawnGroup(
       Masked Float=&Lerp_Float(
        A=2000
        B=50
        Alpha=&FloatFromCurve(FloatCurve=DataInterface() {propertyValues {Curve {keys [
             {interpMode=RCIM_Cubic, leaveTangentWeight=0.1335, tangentMode=RCTM_User, time=0, value=0}
             {interpMode=RCIM_Cubic, tangentMode=RCTM_User, time=0.4, value=1}
            ]}}})
       )
       Passthrough Float=0
       Spawn Group Mask=2
      )
      Spawn Group Mask=1
     )
    )
    Pan Noise=false
   ) {module=CurlNoiseForce001}
   &SolveForcesAndVelocity(Clamp Velocity=false, Limit Acceleration=false)
   &ScaleSpriteSize(Scale Factor=&Vector2DFromFloat(Value=&MaskFloatBySpawnGroup(
      Masked Float=&FloatFromCurve(FloatCurve=DataInterface() {propertyValues {Curve {keys [
           {arriveTangent=2.4809, interpMode=RCIM_Cubic, leaveTangent=2.4809, tangentMode=RCTM_User, time=0, value=0.5}
           {arriveTangent=0.1572, interpMode=RCIM_Cubic, leaveTangent=0.1572, tangentMode=RCTM_User, time=1, value=1.8}
          ]}}})
      Passthrough Float=&MaskFloatBySpawnGroup(
       Masked Float=&FloatFromCurve(FloatCurve=DataInterface() {propertyValues {Curve {keys [
            {interpMode=RCIM_Cubic, tangentMode=RCTM_User, time=0, value=0}
            {interpMode=RCIM_Cubic, tangentMode=RCTM_User, time=0.5, value=1}
            {tangentMode=RCTM_User, time=1, value=0}
           ]}}})
       Passthrough Float=&MaskFloatBySpawnGroup(
        Masked Float=&FloatFromCurve(
         FloatCurve=DataInterface() {propertyValues {Curve {keys [
             {interpMode=RCIM_Cubic, tangentMode=RCTM_User, time=0, value=1}
             {arriveTangent=-3.4824, interpMode=RCIM_Cubic, leaveTangent=-3.4824, tangentMode=RCTM_User, time=0.2, value=2.5}
             {interpMode=RCIM_Cubic, tangentMode=RCTM_User, time=1, value=0}
            ]}}}
         Scale Curve=0.25
        )
        Passthrough Float=&MakeFloatFromVector2d(Vector 2D=&Lerp_Vector2(A=<1 1>, B=<0 0.5>, Alpha=$Particles.NormalizedAge))
        Spawn Group Mask=2
       )
       Spawn Group Mask=1
      )
     )))
   &ScaleColor(
    Scale Alpha=&MaskFloatBySpawnGroup(
     Masked Float=&FloatFromCurve(
      FloatCurve=DataInterface() {propertyValues {Curve {keys [
          {arriveTangent=9.0725, interpMode=RCIM_Cubic, leaveTangent=9.0725, tangentMode=RCTM_User, time=0, value=0}
          {arriveTangent=-2.5249, interpMode=RCIM_Cubic, leaveTangent=-2.5249, tangentMode=RCTM_User, time=0.1408, value=0.9847}
          {arriveTangent=-0.0405, interpMode=RCIM_Cubic, leaveTangent=-0.0405, tangentMode=RCTM_User, time=1, value=0}
         ]}}}
      Scale Curve=0.03
     )
     Passthrough Float=&MaskFloatBySpawnGroup(
      Masked Float=&FloatFromCurve(
       FloatCurve=DataInterface() {propertyValues {Curve {keys [
           {arriveTangent=9.0725, interpMode=RCIM_Cubic, leaveTangent=9.0725, tangentMode=RCTM_User, time=0, value=0}
           {arriveTangent=-5.8355, interpMode=RCIM_Cubic, leaveTangent=-5.8355, tangentMode=RCTM_User, time=0.1408, value=2}
           {arriveTangent=-0.0405, interpMode=RCIM_Cubic, leaveTangent=-0.0405, tangentMode=RCTM_User, time=1, value=0}
          ]}}}
       Scale Curve=0.04
      )
      Passthrough Float=&MaskFloatBySpawnGroup(
       Masked Float=1
       Passthrough Float=&FloatFromCurve(FloatCurve=DataInterface() {propertyValues {Curve {keys [{arriveTangent=11.9433, interpMode=RCIM_Cubic, time=0, value=0}, <at 1 1 RCIM_Cubic>]}}})
       Spawn Group Mask=2
      )
      Spawn Group Mask=1
     )
    )
    Scale RGB=&MaskVectorBySpawnGroup(
     Masked Vector=&VectorFromFloat(Value=&FloatFromCurve(FloatCurve=DataInterface() {propertyValues {Curve {keys [<at 0 1>,<at 0.1 3>,<at 0.2 1>,<at 1 1>]}}}))
     Passthrough Vector=&MaskVectorBySpawnGroup(
      Masked Vector=&VectorFromFloat(Value=&FloatFromCurve(FloatCurve=DataInterface() {propertyValues {Curve {keys [
            {arriveTangent=9.0725, interpMode=RCIM_Cubic, leaveTangent=9.0725, tangentMode=RCTM_User, time=0, value=1}
            {arriveTangent=-17.1917, interpMode=RCIM_Cubic, leaveTangent=-17.1917, tangentMode=RCTM_User, time=0.1, value=10}
            {arriveTangent=-0.0405, interpMode=RCIM_Cubic, leaveTangent=-0.0405, tangentMode=RCTM_User, time=1, value=1}
           ]}}}))
      Passthrough Vector=&MaskVectorBySpawnGroup(Masked Vector=<1 1 1>, Passthrough Vector=&VectorFromFloat(Value=25), Spawn Group Mask=2)
      Spawn Group Mask=1
     )
    )
    ScaleA=true
    ScaleRGB=true
   )
  ]
  rendererClasses [&NiagaraSpriteRendererProperties]
  renderers [
   {rendererClass=&NiagaraSpriteRendererProperties}
   {rendererClass=&NiagaraSpriteRendererProperties, rendererIndex=1}
   {rendererClass=&NiagaraSpriteRendererProperties, rendererIndex=2}
   {rendererClass=&NiagaraSpriteRendererProperties, rendererIndex=3}
  ]
 }
 Confetti {
  emitterUpdateScript [
   &EmitterState(
    Delay First Loop Only=true
    Life Cycle Mode=::NewEnumerator1
    Loop Behavior=::NewEnumerator2
    Loop Count=4
    Loop Delay=0.04
    Loop Duration=0.01
    UseLoopDelay=true
   )
   &SpawnBurst_Instantaneous(
    Spawn Count=&RandomRangeInt(Minimum=65, Maximum=75, Evaluation Type=::NewEnumerator1)
    Use Spawn Probability=false
   )
  ]
  particleSpawnScript [
   &InitializeParticle(
    Color=&Multiply_LinearColorByFloat(
     Linear Color=&SelectLinearColorFromArray(
      Array Sampling Mode=::NewEnumerator0
      Direct Array Index=?
      Color Selection Array=DataInterface() {propertyValues {ColorData [<rgba 1 0.1614 0.9113 1>, <rgba 0.148 0 1 1>, <rgba 0.0772 0.9637 1 1>, <rgba 1 0.9088 0.0365 1>, <rgba 0.9385 0 0 1>]}}
      Random Seed=0
     )
     Float=0.97
    )
    Lifetime=&RandomRangeFloat(Minimum=2, Maximum=3)
    Sprite Size Max=<6 6>
    Sprite Size Min=<2 2>
    Sprite Size Mode=::NewEnumerator2
    UsePositionOffset=false
    Write Material Random=true
   ) {module=InitializeParticle001}
   &ShapeLocation(Box Size=<10 10 10>, MaskSpawnGroup=false, Override Seed=false, Shape Primitive=::NewEnumerator2, Surface Only=false, UseOwnerScale=true)
   &AddVelocity(
    Module.Use Falloff Distance=false
    Origin Offset=<0 0 -1>
    Override Seed=false
    Velocity Mode=::NewEnumerator1
    Velocity Speed=/Niagara/DynamicInputs/Math/LerpMultipleFloats(
     Float Count:ENiagara_MultipleLerpCount=::NewEnumerator0
     1:NiagaraFloat=&RandomRangeFloat(Minimum=50, Maximum=112.5)
     2:NiagaraFloat=&RandomRangeFloat(Minimum=62.5, Maximum=162.5)
     3:NiagaraFloat=&RandomRangeFloat(Minimum=50, Maximum=150)
     4:NiagaraFloat=?
     5:NiagaraFloat=?
     Alpha:NiagaraFloat=&Divide_Float(
      A=&MakeFloatFromInt(Int=3)
      B=&MakeFloatFromInt(Int=&SpawnGroup())
     )
    )
   ) {module=AddVelocity001}
   /SetVariables(Particles.Color=&MakeLinearColorFromVectorAndFloat(
     "Vector (RGB)"=&AddVectorToPosition(
      Position=HlslExpression() {hlslExpression=Particles.Color.xyz}
      Vector=&FN_FrontEndVector3Modifier(Gameplay=<0 0 0>, Front End=<0 0.1 0.25>)
     )
     "Float (Alpha)"=1
    )) {isSetParameters=true}
   &CurlNoiseForce(
    Bias Noise=false
    Noise Strength=&RandomRangeFloat(Minimum=25, Maximum=50)
    Pan Noise=true
    Pan Noise Field=<0.2016 0 0.5008>
   ) {module=CurlNoiseForce001}
   &AddVelocity(
    Module.Use Falloff Distance=false
    Override Seed=false
    Velocity Mode=::NewEnumerator1
    Velocity Speed=&RandomRangeFloat(Minimum=300, Maximum=700)
   )
   &InheritVelocity(
    Inherited Velocity Amount Scale=<1 0 0>
    Inherited Velocity Speed Limit=400
    Speed Limit=true
    UseSpeedThreshold=false
    Velocity Destination Space:ENiagaraCoordinateSpace=::World
   )
   &ApplyInitialForces()
  ]
  particleUpdateScript [
   &ParticleState()
   &ScaleSpriteSize(Scale Factor=&Vector2DFromFloat(Value=&FloatFromCurve(
      FloatCurve=DataInterface() {propertyValues {Curve {keys [
          {arriveTangent=-0.4596, interpMode=RCIM_Cubic, leaveTangent=-0.4596, tangentMode=RCTM_User, time=0.55, value=1}
          {arriveTangent=-4.9031, interpMode=RCIM_Cubic, leaveTangent=-4.9031, tangentMode=RCTM_User, time=1, value=0}
         ]}}}
      Scale Curve=&RandomRangeFloat(Minimum=0.4, Maximum=1, Randomness Mode=::NewEnumerator2, Random Seed=?)
     )))
   &SpriteRotationRate(Rotation Rate=&Multiply_Float(
     A=&SelectFloatFromArray(
      Array Sampling Mode=::NewEnumerator0
      Direct Array Index=?
      Float Selection Array=DataInterface() {propertyValues {FloatData [-300,300]}}
      Random Seed=0
     )
     B=&RandomRangeFloat(Minimum=-5, Maximum=5)
    ))
   &Drag(
    Drag=&FloatFromCurve(
     FloatCurve=DataInterface() {propertyValues {Curve {keys [{interpMode=RCIM_Cubic, tangentMode=RCTM_User, time=0, value=1}, {interpMode=RCIM_Cubic, tangentMode=RCTM_User, time=1, value=0.4}]}}}
     Scale Curve=&RandomRangeFloat(Minimum=5, Maximum=15)
    )
    Use Linear Drag=true
    Use Rotational Drag=false
   )
   &GravityForce(Gravity=&RandomRangeVector(Minimum=<0 0 -150>, Maximum=<1 1 -650>))
   &CurlNoiseForce(Noise Strength=&RandomRangeFloat(Minimum=200, Maximum=300), Pan Noise=false)
   &Collision(
    Age Colliding Particles=false
    CPU Trace Vector Length Multiplier=1
    Enable Rest State=false
    Friction=0.5
    Module.EnableMaxCollisionCount=false
    Module.Randomize Collision Normal Vector=false
    Particle Radius Scale=0
    Restitution=0.2
    Restitution Coefficient Blending:ENiagaraRestitutionMergeType=::NewEnumerator1
    Simple Friction=false
   )
   &SolveForcesAndVelocity(Clamp Velocity=false, Limit Acceleration=false)
   &ScaleSpriteSizeBySpeed(Max Scale Factor=<1 1>, Min Scale Factor=<0 0>, Velocity Threshold=20)
   &KillParticles(Kill Particles=&SetBoolByFloatComparison(A=&VectorLength(Vector=$Particles.Velocity), B=1, Comparison Type=::NewEnumerator6))
  ]
  rendererClasses [&NiagaraSpriteRendererProperties]
  renderers [{rendererClass=&NiagaraSpriteRendererProperties}]
 }
 SpikeBurst {
  emitterUpdateScript [
   &EmitterState(
    Enable Distance Culling=true
    Enable Visibility Culling=true
    Max Distance Response=::KillImmediately
    Module.bUseMaxDistance=true
    Module.bUseMinDistance=false
    Scalability Mode=::NewEnumerator1
   )
   &SpawnBurst_Instantaneous(Spawn Count=11, Use Spawn Probability=false)
   &SpawnBurst_Instantaneous(Spawn Count=9, Spawn Group=1, Use Spawn Probability=false) {module=SpawnBurst_Instantaneous001}
  ]
  particleSpawnScript [
   /SetVariables(Particles.VisibilityTag=&SpawnGroup()) {isSetParameters=true}
   &InitializeParticle(
    Color=<rgba 0.4167 0.3538 0.1674 1>
    Lifetime=&MaskFloatBySpawnGroup(Masked Float=&UniformRangedFloat(Minimum=0.175, Maximum=0.2), Passthrough Float=0.25)
    Mesh Scale=&Multiply_VectorByFloat(Vector=&UniformRangedVector(Minimum=<0.9 0.6 0.3>, Maximum=<1.5 0.6 0.3>), Float=0.475)
    Mesh Scale Mode=::NewEnumerator4
    Sprite Size Max=&Multiply_Vector2D_ByFloat(Float=0.65, Vector2D=<20 50>)
    Sprite Size Min=&Multiply_Vector2D_ByFloat(Float=0.6, Vector2D=<40 62>)
    Sprite Size Mode=::NewEnumerator2
    UsePositionOffset=false
    Write Material Random=true
   ) {module=InitializeParticle001}
   &ShapeLocation(
    MaskSpawnGroup=false
    Override Seed=false
    Sphere Radius=&MaskFloatBySpawnGroup(Masked Float=3, Passthrough Float=5)
    UseOwnerScale=true
   )
   &CameraOffset(Camera Offset Amount=3) {module=CameraOffset001}
   &AddVelocity(
    Module.Use Falloff Distance=false
    Override Seed=false
    Velocity Mode=::NewEnumerator1
    Velocity Speed=&MaskFloatBySpawnGroup(Masked Float=75, Passthrough Float=3)
   ) {module=AddVelocity001}
  ]
  particleUpdateScript [
   &ParticleState()
   &ScaleMeshSize(Scale Factor=&VectorFromCurve(
     VectorCurve=DataInterface() {propertyValues {
       bOptimizeLUT=false
       XCurve {keys [{arriveTangent=1.8635, interpMode=RCIM_Cubic, leaveTangent=1.8635, tangentMode=RCTM_User, time=-0.0048, value=0.4}, <at 0.9737 1>]}
       YCurve {keys [
         {arriveTangent=0.1613, interpMode=RCIM_Cubic, leaveTangent=0.1613, tangentMode=RCTM_User, time=0, value=1}
         {arriveTangent=-1.3776, interpMode=RCIM_Cubic, leaveTangent=-1.3776, tangentMode=RCTM_User, time=0.9944, value=0}
        ]}
       ZCurve {keys [<at 0 1 RCIM_Cubic>, {arriveTangent=-1.635, interpMode=RCIM_Cubic, leaveTangent=-1.635, tangentMode=RCTM_User, time=0.9961, value=0}]}
      }}
     Scale Curve=<0.9 0.9 0.9>
    ))
   &ScaleSpriteSize(Scale Factor=&Vector2DFromFloat(Value=&FloatFromCurve(FloatCurve=DataInterface() {propertyValues {Curve {keys [<at 0 0.6>,<at 1 1.7>]}}}, Scale Curve=0.5)))
   &CameraOffset()
   &SolveForcesAndVelocity(Clamp Velocity=false, Limit Acceleration=false)
   &ScaleColor(
    Scale Alpha=&MaskFloatBySpawnGroup(
     Masked Float=&FloatFromCurve(
      FloatCurve=DataInterface() {propertyValues {Curve {keys [
          {arriveTangent=-4.0245, interpMode=RCIM_Cubic, leaveTangent=-4.0245, tangentMode=RCTM_User, time=0, value=1.5}
          {arriveTangent=-1.5, interpMode=RCIM_Cubic, leaveTangent=-1.5, time=0.2, value=0.8}
          <at 1 0>
         ]}}}
      Scale Curve=0.5
     )
     Passthrough Float=&FloatFromCurve(
      FloatCurve=DataInterface() {propertyValues {Curve {keys [
          {arriveTangent=-4.0245, interpMode=RCIM_Cubic, leaveTangent=-4.0245, tangentMode=RCTM_User, time=0, value=1.5}
          {arriveTangent=-1.5, interpMode=RCIM_Cubic, leaveTangent=-1.5, time=0.2, value=0.8}
          <at 1 0>
         ]}}}
      Scale Curve=0.385
     )
    )
    Scale RGB=&MaskVectorBySpawnGroup(
     Masked Vector=&VectorFromFloat(Value=0.7)
     Passthrough Vector=&VectorFromFloat(Value=&FloatFromCurve(FloatCurve=DataInterface() {propertyValues {Curve {keys [
           {arriveTangent=-4.0245, interpMode=RCIM_Cubic, leaveTangent=-4.0245, tangentMode=RCTM_User, time=0, value=1.5}
           {arriveTangent=-1.5, interpMode=RCIM_Cubic, leaveTangent=-1.5, time=0.2, value=0.8}
           <at 1 0>
          ]}}}))
    )
    ScaleA=true
    ScaleRGB=true
   )
   /Game/Effects/Niagara/Modules/Materials/MaterialPanner(Offset:NiagaraFloat=&Add_Float(A=$Emitter.Age, B=$Particles.Initial.MaterialRandom), Speed=-5)
  ]
  rendererClasses [&NiagaraSpriteRendererProperties, &NiagaraMeshRendererProperties]
  renderers [
   {rendererClass=&NiagaraSpriteRendererProperties}
   {rendererClass=&NiagaraMeshRendererProperties, rendererIndex=1}
  ]
 }
}
summary {
 systemName=NS_Emote_NerdStomp_Confetti_Clap
 emitters [
  {emitterName=Confetti, rendererClasses [&NiagaraSpriteRendererProperties]}
  {emitterName=Burst, rendererClasses [&NiagaraSpriteRendererProperties]}
  {emitterName=SpikeBurst, rendererClasses [&NiagaraSpriteRendererProperties, &NiagaraMeshRendererProperties]}
 ]
 userVariables [
  User.Balloon Color {type=LinearColor, description="", defaultValue {struct=LinearColor, value=<rgba 1 1 1 1>}}
  User.bIsFrontEnd {type=NiagaraBool}
  User.bIsFrontEndPreview {type=NiagaraBool}
 ]
}
