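# Niagara system for soft grey smoke puffs at footfalls during the NerdStomp emote's traversal.
# One looping sprite emitter combines a small burst with a steady spawn rate; alternate rate and per-unit spawners are disabled.
# Box-spawned particles inherit velocity, scatter under local X acceleration, then slow through drag while growing.
# Randomized, age-driven material parameters vary the smoke, and alpha fades over life, scaled by front-end versus gameplay context and time of day.
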NiagaraSystem /BRCosmetics/Animation/Game/MainPlayer/Emotes/NerdStomp/FX/NS_Emote_NerdStomp_TraversalFootFall
dependencies {
 usedDataInterfaces [&NiagaraDataInterfaceCurve]
 usedDynamicInputs [&Vector2DFromFloat]
 usedModules [&SystemState]
}
emitters {Default_Emote_TraversalSmoke_Loop {
  emitterUpdateScript [
   &EmitterState(Life Cycle Mode=::NewEnumerator1, Loop Duration=1, UseLoopDelay=false)
   &SpawnBurst_Instantaneous(
    Spawn Count=&RandomRangeInt(Minimum=7, Maximum=10)
    Use Spawn Probability=false
    UseLoopCountLimit=false
   )
   &SpawnRate(SpawnRate=13, Use Spawn Probability=false) {module=SpawnRate001}
   &SpawnRate(SpawnRate=&FN_FrontEndFloatModifier(Gameplay=0, Front End=44), Use Spawn Probability=false) {enabled=false}
   &SpawnPerUnit(
    Module.Use Max Movement Threshold=true
    Module.Use Movement Tolerance=true
    Movement Tolerance=3
    Spawn Spacing=5
    Use Spawn Probability=false
   ) {enabled=false, module=SpawnPerUnit001}
  ]
  particleSpawnScript [
   &InitializeParticle(
    Color Channel Mode:ENiagaraChannelCorrelation=::NewEnumerator0
    Color Maximum=<rgba 0.2 0.1987 0.192 0.7>
    Color Minimum=<rgba 0.25 0.2432 0.2322 0.5>
    Color Mode=::NewEnumerator2
    Lifetime Max=1.2
    Lifetime Min=0.5
    Lifetime Mode=::NewEnumerator1
    Position Offset=<0 0 2>
    Sprite Rotation Mode=::NewEnumerator2
    Sprite Size Mode=::NewEnumerator1
    Uniform Sprite Size Max=30
    Uniform Sprite Size Min=17
    UsePositionOffset=true
    Write Material Random=true
   ) {module=InitializeParticle001}
   &InheritVelocity(Inherited Velocity Speed Limit=1000, Speed Limit=true, UseSpeedThreshold=false) {module=InheritVelocity001}
   &AddVelocity(
    Velocity=&RandomRangeVector(Minimum=<1 1 0>, Maximum=<-2 -1 1>)
    Velocity Speed Scale=&RandomRangeFloat(Minimum=0, Maximum=100)
   ) {module=AddVelocity001}
   &AccelerationForce(Acceleration=&UniformRangedVector(Minimum=<250 0 0>, Maximum=<-250 0 0>), Coordinate Space=::Local)
   &ApplyInitialForces()
   &ShapeLocation(
    Box Midpoint=<0.75 0.5 0.5>
    Box Size=<2 15 1>
    MaskSpawnGroup=false
    Override Seed=false
    Shape Primitive=::NewEnumerator2
    Surface Only=false
    UseOwnerScale=true
   )
  ]
  particleUpdateScript [
   &ParticleState()
   &Drag(Drag=10, Use Linear Drag=true, Use Rotational Drag=true)
   &SolveForcesAndVelocity(Clamp Velocity=false, Limit Acceleration=false)
   &ScaleSpriteSize(Uniform Curve Sprite Scale=DataInterface() {propertyValues {Curve {keys [{arriveTangent=1.491, interpMode=RCIM_Cubic, leaveTangent=1.491, tangentMode=RCTM_User, time=0, value=0.5}, <at 1 3 RCIM_Cubic>]}}})
   &DynamicMaterialParameters(
    Index 0 Param 1=&Add_Float(
     A=&RandomRangeFloat(Minimum=0, Maximum=1)
     B=&FloatFromCurve(
      FloatCurve=DataInterface() {propertyValues {Curve {keys [
          <at 0 0 RCIM_Cubic>
          {arriveTangent=1.5947, interpMode=RCIM_Cubic, leaveTangent=1.5947, tangentMode=RCTM_User, time=0.2557, value=0.5257}
          <at 1 1 RCIM_Cubic>
         ]}}}
      CurveIndex=$Particles.Age
      Scale Curve=0.25
     )
    )
    Module.Param0WriteEnabled=true
    Module.Param1WriteEnabled=true
    Module.Param2WriteEnabled=true
    Module.Param3WriteEnabled=true
   ) {module=DynamicMaterialParameters001}
   &CameraOffset(Camera Offset Amount=15)
   &ScaleColor(
    Scale Alpha=&Multiply_Float(
     A=&FloatFromCurve(
      FloatCurve=DataInterface() {propertyValues {Curve {keys [{interpMode=RCIM_Cubic, tangentMode=RCTM_User, time=0, value=0}, <at 0.15 1 RCIM_Cubic>, <at 0.9 0>]}}}
      Scale Curve=&FN_FrontEndFloatModifier(Gameplay=0.75, Front End=1)
     )
     B=&FloatFromCurve(
      FloatCurve=DataInterface() {propertyValues {Curve {keys [
          {interpMode=RCIM_Cubic, tangentMode=RCTM_User, time=0, value=0.6}
          {interpMode=RCIM_Cubic, tangentMode=RCTM_User, time=0.5, value=1}
          {tangentMode=RCTM_User, time=1, value=0.6}
         ]}}}
      CurveIndex=Linked() {linkedVariable {name=NPC.FortniteNPC.TimeOfDay_0-1, type=NiagaraFloat}}
     )
    )
    ScaleA=true
    ScaleRGB=false
   )
  ]
  rendererClasses [&NiagaraSpriteRendererProperties]
  renderers [
   {rendererClass=&NiagaraSpriteRendererProperties}
   {rendererClass=&NiagaraSpriteRendererProperties, rendererIndex=1}
  ]
 }}
summary {
 systemName=NS_Emote_NerdStomp_TraversalFootFall
 emitters [{emitterName=Default_Emote_TraversalSmoke_Loop, rendererClasses [&NiagaraSpriteRendererProperties]}]
 userVariables [
  User.bIsFrontEnd {type=NiagaraBool}
  User.bIsFrontEndPreview {type=NiagaraBool}
 ]
}
