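# Niagara system that bursts colored confetti, named for footfalls in the NerdStomp traversal emote.
# Its single sprite emitter, Confetti_Pieces, spawns pieces with one of five array colors, curve-biased random sizes and lifetimes, and mostly +X velocity.
# Size and spin curves make pieces pop, shrink and slow their rotation.
# Heavy drag, early-strong curl noise and randomized curve-driven gravity turn the initial burst into a fluttering drift.
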NiagaraSystem /BRCosmetics/Animation/Game/MainPlayer/Emotes/NerdStomp/FX/NS_Emote_NerdStomp_TraversalFootFall_Confetti
dependencies {
 usedDataInterfaces [
  &NiagaraDataInterfaceCurve
  &NiagaraDataInterfaceArrayColor
  &NiagaraDataInterfaceVector2DCurve
  &NiagaraDataInterfaceVectorCurve
 ]
 usedDynamicInputs [&Vector2DFromCurve]
 usedModules [&SystemState]
}
emitters {Confetti_Pieces {
  emitterUpdateScript [
   &EmitterState()
   &SpawnBurst_Instantaneous(
    Spawn Count=&RandomRangeInt(Minimum=55, Maximum=60)
    Use Spawn Probability=false
    UseLoopCountLimit=false
   )
  ]
  particleSpawnScript [
   &InitializeParticle(
    Color=&Multiply_LinearColorByFloat(
     Linear Color=&SelectLinearColorFromArray(
      Array Sampling Mode=::NewEnumerator0
      Direct Array Index=?
      Color Selection Array=DataInterface() {propertyValues {ColorData [<rgba 1 0.1614 0.9113 1>, <rgba 0.148 0 1 1>, <rgba 0.0772 0.9637 1 1>, <rgba 1 0.9088 0.0365 1>, <rgba 0.9385 0 0 1>]}}
      Random Seed=0
     )
     Float=0.97
    )
    Lifetime=&Lerp_Float(
     A=0.25
     B=1.2
     Alpha=&FloatFromCurve(
      FloatCurve=DataInterface() {propertyValues {Curve {keys [
          {interpMode=RCIM_Cubic, tangentMode=RCTM_User, time=0, value=0}
          {arriveTangent=2.9148, interpMode=RCIM_Cubic, leaveTangent=2.9148, tangentMode=RCTM_User, time=1, value=1}
         ]}}}
      CurveIndex=&RandomRangeFloat(Minimum=0, Maximum=1)
     )
    )
    Position Offset=<0 0 -5>
    Sprite Rotation Mode=::NewEnumerator2
    Sprite Size=&Multiply_Vector2D_ByFloat(
     Float=1
     Vector2D=&MakeVector2D(
      X=&RandomRangeFloat(Minimum=8, Maximum=15)
      Y=&Lerp_Float(
       A=12
       B=50
       Alpha=&FloatFromCurve(
        FloatCurve=DataInterface() {propertyValues {Curve {keys [
            {interpMode=RCIM_Cubic, tangentMode=RCTM_User, time=0, value=0}
            {arriveTangent=1, interpMode=RCIM_Cubic, leaveTangent=1, time=0.8, value=0.3}
            {arriveTangent=6.5099, arriveTangentWeight=0.7128, interpMode=RCIM_Cubic, leaveTangent=6.5099, tangentMode=RCTM_User, time=1, value=1}
           ]}}}
        CurveIndex=&RandomRangeFloat(Minimum=0, Maximum=1)
       )
      )
     )
    )
    Sprite Size Mode=::NewEnumerator4
    UsePositionOffset=true
    Write Material Random=true
   )
   &ShapeLocation(
    Box Size=<1 20 5>
    MaskSpawnGroup=false
    Override Seed=false
    Shape Primitive=::NewEnumerator2
    Surface Expansion Mode:ENiagaraExpansionMode=::NewEnumerator2
    Surface Only=true
    Surface Only Band Thickness=1
    UseOwnerScale=true
   )
   &AddVelocity(
    Velocity=&RandomRangeVector(Minimum=<1 -0.3 -0.7>, Maximum=<1.2 1.75 0.7>)
    Velocity Speed Scale=&Lerp_Float(
     A=100
     B=200
     Alpha=&FloatFromCurve(
      FloatCurve=DataInterface() {propertyValues {Curve {keys [
          {interpMode=RCIM_Cubic, tangentMode=RCTM_User, time=0, value=0}
          {arriveTangent=2.9148, interpMode=RCIM_Cubic, leaveTangent=2.9148, tangentMode=RCTM_User, time=1, value=1}
         ]}}}
      CurveIndex=&RandomRangeFloat(Minimum=0, Maximum=1)
     )
    )
   )
  ]
  particleUpdateScript [
   &ParticleState()
   &ScaleSpriteSize(
    Uniform Curve Scale=0.175
    Uniform Curve Sprite Scale=DataInterface() {propertyValues {Curve {keys [
        {arriveTangent=34.2693, interpMode=RCIM_Cubic, leaveTangent=34.2693, tangentMode=RCTM_User, time=0, value=-0.25}
        {arriveTangent=-2.0397, interpMode=RCIM_Cubic, leaveTangent=-2.0397, tangentMode=RCTM_User, time=0.2, value=1.7}
        {arriveTangent=-1.5882, interpMode=RCIM_Cubic, leaveTangent=-1.5882, tangentMode=RCTM_User, time=1, value=0}
       ]}}}
   )
   &SpriteRotationRate(Rotation Rate=&Multiply_Float(
     A=&Lerp_Float(
      A=25
      B=1000
      Alpha=&FloatFromCurve(FloatCurve=DataInterface() {propertyValues {Curve {keys [
           {arriveTangent=-3.2581, interpMode=RCIM_Cubic, leaveTangent=-3.2581, tangentMode=RCTM_User, time=0.1, value=1}
           {interpMode=RCIM_Cubic, tangentMode=RCTM_User, time=1, value=0}
          ]}}})
     )
     B=&FloatFromCurve(
      FloatCurve=DataInterface() {propertyValues {Curve {keys [<at 0 -1 RCIM_Constant>,<at 0.5 1 RCIM_Constant>]}}}
      CurveIndex=&RandomRangeFloat(Minimum=-1, Maximum=1)
      Scale Curve=&RandomRangeFloat(Minimum=-2, Maximum=2)
     )
    ))
   &AccelerationForce(Acceleration=&RandomRangeVector(Minimum=<15 35 5>, Maximum=<-5 -5 -5>))
   &Drag(Drag=9, Use Linear Drag=true, Use Rotational Drag=false) {module=Drag001}
   &CurlNoiseForce(
    Noise Strength=&Lerp_Float(
     A=1000
     B=50
     Alpha=&FloatFromCurve(FloatCurve=DataInterface() {propertyValues {Curve {keys [
          {interpMode=RCIM_Cubic, leaveTangentWeight=0.1335, tangentMode=RCTM_User, time=0, value=0}
          {interpMode=RCIM_Cubic, tangentMode=RCTM_User, time=0.4, value=1}
         ]}}})
    )
    Pan Noise=false
   ) {module=CurlNoiseForce001}
   &GravityForce(Gravity=&VectorFromCurve(
     VectorCurve=DataInterface() {propertyValues {
       XCurve {keys [{tangentMode=RCTM_User, time=0, value=1}, {tangentMode=RCTM_User, time=1, value=0}]}
       YCurve {keys [{tangentMode=RCTM_User, time=0, value=0}, {tangentMode=RCTM_User, time=1, value=1}]}
       ZCurve {keys [
         {arriveTangent=-3.0122, interpMode=RCIM_Cubic, leaveTangent=-3.0122, tangentMode=RCTM_User, time=0, value=0.8}
         {arriveTangent=3.8404, interpMode=RCIM_Cubic, leaveTangent=3.8404, tangentMode=RCTM_User, time=0.2, value=0.1}
         {arriveTangent=0.0224, interpMode=RCIM_Cubic, leaveTangent=0.0224, tangentMode=RCTM_User, time=1, value=1}
        ]}
      }}
     Scale Curve=&RandomRangeVector(Minimum=<5 -100 -100>, Maximum=<100 100 -300>)
    ))
   &SolveForcesAndVelocity(Clamp Velocity=false, Limit Acceleration=false)
  ]
  rendererClasses [&NiagaraSpriteRendererProperties]
  renderers [{rendererClass=&NiagaraSpriteRendererProperties}]
 }}
summary {
 systemName=NS_Emote_NerdStomp_TraversalFootFall_Confetti
 emitters [{emitterName=Confetti_Pieces, rendererClasses [&NiagaraSpriteRendererProperties]}]
 userVariables [
  User.bIsFrontEnd {type=NiagaraBool}
  User.bIsFrontEndPreview {type=NiagaraBool}
 ]
}
