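# Niagara confetti effect for the NerdStomp emote (Roll variant), built from a single sprite emitter.
# A random burst plus a spawn rate emits particles in five random colors, brightened by a front-end/gameplay modifier.
# Layered launch velocities, curl noise, curved drag, random gravity, collision and inherited emitter velocity drive tumbling, falling motion.
# Sprites shrink over life and with low speed, and particles are killed by a velocity threshold test.
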NiagaraSystem /BRCosmetics/Animation/Game/MainPlayer/Emotes/NerdStomp/FX/NS_LFN_Emote_NerdStomp_Confetti_Roll
dependencies {
 usedDataInterfaces [&NiagaraDataInterfaceArrayColor, &NiagaraDataInterfaceCurve, &NiagaraDataInterfaceArrayFloat]
 usedDynamicInputs []
 usedModules [&SystemState]
}
emitters {Confetti {
  emitterUpdateScript [
   &EmitterState()
   &SpawnRate(SpawnRate=1000, Use Spawn Probability=false)
   &SpawnBurst_Instantaneous(
    Spawn Count=&RandomRangeInt(Minimum=45, Maximum=55, Evaluation Type=::NewEnumerator1)
    Use Spawn Probability=false
   )
  ]
  particleSpawnScript [
   &InitializeParticle(
    Color=&Multiply_LinearColorByFloat(
     Linear Color=&SelectLinearColorFromArray(
      Array Sampling Mode=::NewEnumerator0
      Direct Array Index=?
      Color Selection Array=DataInterface() {propertyValues {ColorData [<rgba 1 0.1614 0.9113 1>, <rgba 0.148 0 1 1>, <rgba 0.0772 0.9637 1 1>, <rgba 1 0.9088 0.0365 1>, <rgba 0.9385 0 0 1>]}}
      Random Seed=0
     )
     Float=0.97
    )
    Lifetime=&RandomRangeFloat(Minimum=2, Maximum=4)
    Sprite Size Max=<6 6>
    Sprite Size Min=<2 2>
    Sprite Size Mode=::NewEnumerator2
    UsePositionOffset=false
    Write Material Random=true
   ) {module=InitializeParticle001}
   &ShapeLocation(MaskSpawnGroup=false, Override Seed=false, Sphere Radius=1, UseOwnerScale=true)
   &AddVelocity(
    Module.Use Falloff Distance=false
    Origin Offset=<0 0 -1>
    Override Seed=false
    Velocity Mode=::NewEnumerator1
    Velocity Speed=/Niagara/DynamicInputs/Math/LerpMultipleFloats(
     Float Count:ENiagara_MultipleLerpCount=::NewEnumerator0
     1:NiagaraFloat=&RandomRangeFloat(Minimum=50, Maximum=112.5)
     2:NiagaraFloat=&RandomRangeFloat(Minimum=62.5, Maximum=162.5)
     3:NiagaraFloat=&RandomRangeFloat(Minimum=50, Maximum=150)
     4:NiagaraFloat=?
     5:NiagaraFloat=?
     Alpha:NiagaraFloat=&Divide_Float(
      A=&MakeFloatFromInt(Int=3)
      B=&MakeFloatFromInt(Int=&SpawnGroup())
     )
    )
   ) {module=AddVelocity001}
   /SetVariables(Particles.Color=&MakeLinearColorFromVectorAndFloat(
     "Vector (RGB)"=&AddVectorToPosition(
      Position=HlslExpression() {hlslExpression=Particles.Color.xyz}
      Vector=&FN_FrontEndVector3Modifier(Gameplay=<0 0 0>, Front End=<0 0.1 0.25>)
     )
     "Float (Alpha)"=1
    )) {isSetParameters=true}
   &CurlNoiseForce(
    Bias Noise=false
    Noise Strength=&RandomRangeFloat(Minimum=25, Maximum=50)
    Pan Noise=true
    Pan Noise Field=<0.2016 0 0.5008>
   ) {module=CurlNoiseForce001}
   &AddVelocity(Velocity=<0 -1 -1>, Velocity Speed Scale=&RandomRangeFloat(Minimum=25, Maximum=300))
   &ApplyInitialForces()
  ]
  particleUpdateScript [
   &ParticleState()
   &ScaleSpriteSize(Scale Factor=&Vector2DFromFloat(Value=&FloatFromCurve(
      FloatCurve=DataInterface() {propertyValues {Curve {keys [
          {arriveTangent=-0.4596, interpMode=RCIM_Cubic, leaveTangent=-0.4596, tangentMode=RCTM_User, time=0.55, value=1}
          {arriveTangent=-4.9031, interpMode=RCIM_Cubic, leaveTangent=-4.9031, tangentMode=RCTM_User, time=1, value=0}
         ]}}}
      Scale Curve=&RandomRangeFloat(Minimum=0.4, Maximum=1, Randomness Mode=::NewEnumerator2, Random Seed=?)
     )))
   &SpriteRotationRate(Rotation Rate=&Multiply_Float(
     A=&SelectFloatFromArray(
      Array Sampling Mode=::NewEnumerator0
      Direct Array Index=?
      Float Selection Array=DataInterface() {propertyValues {FloatData [-300,300]}}
      Random Seed=0
     )
     B=&RandomRangeFloat(Minimum=-5, Maximum=5)
    ))
   &Drag(
    Drag=&FloatFromCurve(
     FloatCurve=DataInterface() {propertyValues {Curve {keys [
         {arriveTangent=2.7014, interpMode=RCIM_Cubic, leaveTangent=2.7014, tangentMode=RCTM_User, time=0, value=0.3}
         {arriveTangent=1.1179, interpMode=RCIM_Cubic, leaveTangent=1.1179, tangentMode=RCTM_User, time=0.2, value=0.8}
         {arriveTangent=-0.5, interpMode=RCIM_Cubic, leaveTangent=-0.5, time=0.5, value=0.925}
         {arriveTangent=-0.9392, interpMode=RCIM_Cubic, leaveTangent=-0.9392, tangentMode=RCTM_User, time=1, value=0.4}
        ]}}}
     Scale Curve=&RandomRangeFloat(Minimum=5, Maximum=15)
    )
    Use Linear Drag=true
    Use Rotational Drag=false
   )
   &GravityForce(Gravity=&RandomRangeVector(Minimum=<0 0 -150>, Maximum=<1 1 -650>))
   &CurlNoiseForce(Noise Strength=&RandomRangeFloat(Minimum=200, Maximum=300), Pan Noise=false)
   &Collision(
    Age Colliding Particles=false
    CPU Trace Vector Length Multiplier=1
    Enable Rest State=false
    Friction=0.5
    Module.EnableMaxCollisionCount=false
    Module.Randomize Collision Normal Vector=false
    Particle Radius Scale=0
    Restitution=0.2
    Restitution Coefficient Blending:ENiagaraRestitutionMergeType=::NewEnumerator1
    Simple Friction=false
   )
   &InheritVelocity(
    Inherited Velocity Amount Scale=<-1 -1 -1>
    Inherited Velocity Speed Limit=10
    Source Speed Threshold=10
    Speed Limit=true
    UseSpeedThreshold=true
   ) {module=InheritVelocity001}
   &SolveForcesAndVelocity(Clamp Velocity=false, Limit Acceleration=false)
   &ScaleSpriteSizeBySpeed(Max Scale Factor=<1 1>, Min Scale Factor=<0 0>, Velocity Threshold=20)
   &KillParticles(Kill Particles=&SetBoolByFloatComparison(A=&VectorLength(Vector=$Particles.Velocity), B=1, Comparison Type=::NewEnumerator6))
  ]
  rendererClasses [&NiagaraSpriteRendererProperties]
  renderers [{rendererClass=&NiagaraSpriteRendererProperties}]
 }}
summary {
 systemName=NS_LFN_Emote_NerdStomp_Confetti_Roll
 emitters [{emitterName=Confetti, rendererClasses [&NiagaraSpriteRendererProperties]}]
 userVariables [
  User.bIsFrontEnd {type=NiagaraBool}
  User.bIsFrontEndPreview {type=NiagaraBool}
 ]
}
